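# Simple parameter-driven translucent deferred decal master material for effects, with cracked-decal and ripple textures as defaults.
# A center-scaled texture is refined and tinted by decal color for emissive.
# The opacity chain combines a channel-selected mask, optional radial gradient and decal lifetime fade.
# An optional second mask multiplies in, with rotation, uniform scale and offset, where decal color alpha can drive the U offset.
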Material /Game/Effects/Niagara/MasterMaterials/Decal/M_DecalSimple
graph {
 properties {
  bEnableResponsiveAA=true
  bOutputTranslucentVelocity=true
  bUsedWithNiagaraRibbons=true
  bUsedWithNiagaraSprites=true
  bUsedWithParticleSprites=true
  blendMode=BLEND_Translucent
  materialDomain=MD_DeferredDecal
  pixelDepthOffsetMode=PDOM_Legacy
  shadingModel=MSM_Unlit
  twoSided=true
  usedShadingModels=Unlit
 }
 nodes [
  n1 = TextureSampleParameter2D(UVs=MaterialFunctionCall(Texture Scale=ScalarParameter() {defaultValue=0.6, parameterName=EmissiveUVScale}) {materialFunction=&ScaleUVsByCenter}.UVs) {
   parameterName=Emissive
   samplerSource=SSM_Clamp_WorldGroupSettings
   texture=/Game/Effects/Fort_Effects/Gameplay/SmokeDustExplosions/Textures/T_Decal_Cracked
  }
  n2 = VectorParameter() {parameterName=Emissive Refine, sortPriority=4, defaultValue=<rgba 1 1 1 0.5>}
  MaterialFunctionCall(Emissive=Multiply(
    A=DecalColor().RGB
    B=MaterialFunctionCall(Texture=@n1.R, RGB=@n2.RGB, A=@n2.RGB) {materialFunction=&MF_FX_Refine}.Result
   )) {materialFunction=&MF_FX_EmissiveAdjustments}
  n3 = TextureSampleParameter2D() {
   parameterName=Mask
   samplerSource=SSM_Clamp_WorldGroupSettings
   samplerType=SAMPLERTYPE_LinearColor
   texture=/Game/Effects/Fort_Effects/Effects/Water/Textures/WaterDecalRippleMask
  }
  n4 = Power(
   Base=ChannelMaskParameter(None=@n3.RGBA) {parameterName=Channel Mask Parameter}.R
   Exp=ScalarParameter() {defaultValue=1, parameterName=MaskPower}
  )
  n5 = MaterialFunctionCall(
   Radius=ScalarParameter() {defaultValue=0.6, parameterName="Radius (S)"}
   Density=ScalarParameter() {defaultValue=1, parameterName="Density (S)"}
   Invert Density=StaticBool()
  ) {materialFunction=&RadialGradientExponential}
  n6 = Multiply(A=@n4, B=Multiply(A=Multiply(A=@n5.RadialGradientExponential)) {constB=3}) {constB=2}
  n7 = ScalarParameter() {parameterName=NoneUseAddRadiusMask}
  n8 = NamedRerouteDeclaration(None=Multiply(A=LinearInterpolate(A=@n6, B=@n4, Alpha=@n7), B=ScalarParameter() {defaultValue=1.5, parameterName=MaskMultiply}) {constB=1.5}) {name=Mask, nodeColor=<rgba 0.2188 0 1 1>}
  n9 = NamedRerouteUsage()
  Multiply(
   A=LinearInterpolate(A=@n8, B=Multiply(A=@n8, B=@n9), Alpha=ScalarParameter() {group=02Mask, parameterName=Use02Mask})
   B=DecalLifetimeOpacity().Opacity
  )
  n10 = MaterialFunctionCall(UVs=TextureCoordinate(), "Rotation Angle (0-1)"=ScalarParameter() {group=02Mask, parameterName=02MaskRotation}) {materialFunction=&CustomRotator}
  n11 = VectorParameter() {group=02Mask, parameterName=02MaskUV, defaultValue=<rgba 1 1 0 0>}
  n12 = MaterialFunctionCall(X=@n11.R, Y=@n11.R) {materialFunction=&MakeFloat2}
  n13 = LinearInterpolate(A=@n11.B, B=Multiply(A=@n11.B, B=DecalColor().A), Alpha=ScalarParameter() {group=02Mask, parameterName=UseAlpha02MaskOffset})
  n14 = TextureSampleParameter2D(UVs=Add(
    A=Multiply(A=@n10.Rotated Values, B=@n12.Result)
    B=MaterialFunctionCall(X=@n13, Y=@n11.A) {materialFunction=&MakeFloat2}.Result
   )) {
   group=02Mask
   parameterName=02MaskTex
   samplerSource=SSM_Clamp_WorldGroupSettings
   samplerType=SAMPLERTYPE_LinearColor
   texture=&WaterDecalRippleMask
  }
  NamedRerouteDeclaration(None=Multiply(A=@n14.R, B=ScalarParameter() {defaultValue=1.5, group=02Mask, parameterName=02MaskMultiply}) {constB=1.5}) {name=02Mask, nodeColor=<rgba 1 0 0.4453 1>}
  n15 = TextureSampleParameter2D() {parameterName=Normal, samplerType=SAMPLERTYPE_Normal, texture=/Game/Effects/Fort_Effects/Textures/GroundCrack/T_Decal_GroundCrack_N}
  Reroute(None=MaterialFunctionCall(Normal=@n15.RGB, Flatness=ScalarParameter() {parameterName="Flatness (S)"}) {materialFunction=&FlattenNormal}.Result)
  ScalarParameter() {parameterName=BaseColor}
  ScalarParameter() {defaultValue=0.5, parameterName=Specular}
  ScalarParameter() {parameterName=Metallic}
  ScalarParameter() {defaultValue=1, parameterName=Roughness}
 ]
}
