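# Translucent lit deferred decal master material for Niagara trail effects, with water ripple and wave textures as defaults.
# A shape mask, optionally modulated by a panning radial-UV texture, minus noise erosion driven by decal color red, scales tint and alpha-faded opacity.
# Erosion and flattened normals can use world-aligned or UV sampling.
# Scrolling tiling and speed use only their X values on both axes; emissive color is optional.
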Material /Game/Effects/Niagara/MasterMaterials/Decal/M_Decal_TrailMaster
graph {
 properties {blendMode=BLEND_Translucent, materialDomain=MD_DeferredDecal, pixelDepthOffsetMode=PDOM_Legacy, usedShadingModels=Default Lit}
 nodes [
  LinearInterpolate(
   B=MaterialFunctionCall(Emissive=VectorParameter() {group=Emissive, parameterName=EMISSIVE COLOR, defaultValue=<rgba 0 0 0 1>}.RGB) {materialFunction=&MF_FX_EmissiveAdjustments}.Result
   Alpha=ScalarParameter() {group=Emissive, parameterName=USE EMISSIVE COLOR}
  )
  n1 = TextureSampleParameter2D() {
   group=Mask
   parameterName=Base Shape Texture
   samplerType=SAMPLERTYPE_LinearColor
   sortPriority=1
   texture=/Game/Effects/Fort_Effects/Effects/Water/Textures/WaterDecalRippleMask
  }
  n2 = TextureCoordinate()
  n3 = LinearInterpolate(
   A=@n2
   B=MaterialFunctionCall(Vector or UVs=Subtract(A=@n2) {constB=0.5}) {materialFunction=&VectorToRadialValue}.Radial Coordinates
   Alpha=ScalarParameter() {defaultValue=1, parameterName=Scrolling_Texture_RadialUV}
  ) {constAlpha=1}
  n4 = VectorParameter() {parameterName=Scrolling_Tex_UVSPeed, channelNames {a=Speed Y, b=Speed X, g=Tiling Y, r=Tiling X}, defaultValue=<rgba 1 1 0 1>}
  n5 = AppendVector(A=@n4.Tiling X, B=@n4.Tiling X)
  n6 = Constant()
  n7 = TextureSampleParameter2D!(
   UVs=Panner(Coordinate=Multiply(A=@n3, B=@n5), Speed=AppendVector(A=@n4.Speed X, B=@n4.Speed X))
   "DDX(UVs)"=@n6
   "DDY(UVs)"=@n6
   Apply View MipBias
  ) {
   mipValueMode=TMVM_Derivative
   parameterName=Scrolling_Texture
   samplerType=SAMPLERTYPE_LinearGrayscale
   texture=/Game/Global/Textures/Default/Caustics/T_Nosie_Caustics_Grayscale
  }
  n8 = Saturate(None=Power(Base=@n7.R, Exp=ScalarParameter() {defaultValue=1, parameterName=Scrolling_Tex_Power}))
  n9 = TextureObjectParameter() {
   group=Mask
   parameterName=World Aligned Erosion Texture
   samplerType=SAMPLERTYPE_Grayscale
   sortPriority=2
   texture=&LowResBlurredNoise
  }
  n10 = ScalarParameter() {defaultValue=256, group=Mask, parameterName=World Aligned Erosion Tiling}
  n11 = TextureSampleParameter2D!(
   UVs=Multiply(A=ScalarParameter() {defaultValue=256, group=Mask, parameterName=Erosion Tiling, sortPriority=3}, B=TextureCoordinate())
   Apply View MipBias
  ) {
   group=Mask
   parameterName=Erosion Texture
   samplerType=SAMPLERTYPE_Grayscale
   sortPriority=2
   texture=&LowResBlurredNoise
  }
  n12 = Power(
   Base=StaticSwitchParameter(
    True=MaterialFunctionCall(TextureObject=@n9, TextureSize=@n10) {materialFunction=&WorldAlignedTexture}.XYZ Texture
    False=@n11.R
   ) {defaultValue=true, parameterName=WorldAligned, sortPriority=2}
   Exp=ScalarParameter() {defaultValue=1, group=Mask, parameterName=Erosion Power, sortPriority=4}
  )
  n13 = DecalColor()
  n14 = Subtract(
   A=LinearInterpolate(A=@n1.G, B=Multiply(A=@n1.G, B=@n8), Alpha=ScalarParameter() {parameterName=Scrolling_Tex}) {constAlpha=1}
   B=Saturate(None=Add(A=@n12, B=@n13.R) {constB=-0.3})
  )
  Multiply(A=VectorParameter() {parameterName=Color, sortPriority=1, defaultValue=<rgba 0.1276 0.1701 0.1939 0.2251>}.RGB, B=@n14)
  n15 = TextureObjectParameter() {
   group=Normal
   parameterName=Normal Texture World Aligned
   samplerType=SAMPLERTYPE_Normal
   sortPriority=1
   texture=/Game/Packages/Fortnite_UniqueMaterials_1/Textures/T_Water_TilingNormal_Waves_01
  }
  n16 = MaterialFunctionCall(
   TextureObject=@n15
   TextureSize=ScalarParameter() {defaultValue=600, group=Normal, parameterName=Normal Tiling World Aligned, sortPriority=2}
  ) {materialFunction=&WorldAlignedTexture}
  n17 = TextureCoordinate()
  n18 = ComponentMask(None=VectorParameter() {group=Normal, parameterName=Normal Tiling, sortPriority=2, defaultValue=<rgba 0.25 0.25 0 0>}.RGB) {g=true, r=true}
  n19 = ScalarParameter() {group=Normal, parameterName=Normal Pan X, sortPriority=3}
  n20 = ScalarParameter() {group=Normal, parameterName=Normal Pan Y, sortPriority=4}
  n21 = TextureSampleParameter2D(UVs=Panner(Coordinate=Multiply(A=@n17, B=@n18), Speed=AppendVector(A=@n19, B=@n20))) {
   group=Normal
   parameterName=Normal Texture
   samplerType=SAMPLERTYPE_Normal
   sortPriority=1
   texture=&T_Water_TilingNormal_Waves_01
  }
  Multiply(
   A=@n1.A
   B=MaterialFunctionCall(
    Normal=StaticSwitchParameter(True=@n16.XYZ Texture, False=@n21.RGB) {defaultValue=true, parameterName=WorldAligned, sortPriority=2}
    Flatness=ScalarParameter() {defaultValue=0.7, group=Normal, parameterName=Flatten Normal, sortPriority=5}
   ) {materialFunction=&FlattenNormal}.Result
  )
  Multiply(A=DecalColor().A, B=@n14)
  ScalarParameter() {parameterName=Metallic, sortPriority=3}
  ScalarParameter() {parameterName=Roughness, sortPriority=5}
  ScalarParameter() {defaultValue=1, parameterName=Specular, sortPriority=4}
 ]
}
